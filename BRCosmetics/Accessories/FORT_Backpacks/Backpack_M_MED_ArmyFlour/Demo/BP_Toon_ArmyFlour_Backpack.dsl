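; Demo Blueprint for the ArmyFlour toon backpack, assembled from body, head and face-accessory static meshes, each paired with an outline component.
; The construction script assigns the meshes, caches their original materials, mirrors each mesh onto its outline component and applies per-slot outline materials.
; Debug visibility flags affect only the body outline and the original face accessory, not the other components.
; BeginPlay is empty, so the script has no runtime logic.
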(event EventBeginPlay)

(fn ConstructionScript
  (Utilities|FlowControl|Sequence
    (:then_0
      (Components|StaticMesh|SetStaticMesh (Variables|Default|GetSM_Body) (Variables|Meshes|GetBodyMesh))
      (Components|StaticMesh|SetStaticMesh (Variables|Default|GetSM_Head) (Variables|Meshes|GetHeadMesh))
      (Components|StaticMesh|SetStaticMesh (Variables|Default|GetSM_FaceAcc) (Variables|Meshes|GetFaceAccMesh)))
    (:then_1
      (Variables|OriginalMaterials|SetBodyMaterials (Rendering|Material|GetMaterials (Variables|Default|GetSM_Body)))
      (Variables|OriginalMaterials|SetHeadMaterials (Rendering|Material|GetMaterials (Variables|Default|GetSM_Head)))
      (Variables|OriginalMaterials|SetFaceAccMaterials (Rendering|Material|GetMaterials (Variables|Default|GetSM_FaceAcc))))
    (:then_2
      (Components|StaticMesh|SetStaticMesh (Variables|Default|GetSM_Body_Outline) (Variables|/Script/Engine.StaticMeshComponent|GetStaticMesh (Variables|Default|GetSM_Body)))
      (Components|StaticMesh|SetStaticMesh (Variables|Default|GetSM_Head_Outline) (Variables|/Script/Engine.StaticMeshComponent|GetStaticMesh (Variables|Default|GetSM_Head)))
      (Components|StaticMesh|SetStaticMesh (Variables|Default|GetSM_FaceAcc_Outline) (Variables|/Script/Engine.StaticMeshComponent|GetStaticMesh (Variables|Default|GetSM_FaceAcc))))
    (:then_3
      (for (_array_element _array_index) (Variables|OutlineMaterials|GetBodyOutlineMaterials)
        (Rendering|Material|SetMaterial (Variables|Default|GetSM_Body_Outline) _array_index _array_element))
      (for (_array_element_2 _array_index_2) (Variables|OutlineMaterials|GetHeadOutlineMaterials)
        (Rendering|Material|SetMaterial (Variables|Default|GetSM_Head_Outline) _array_index_2 _array_element_2))
      (for (_array_element_3 _array_index_3) (Variables|OutlineMaterials|GetFaceAccOutlineMaterials)
        (Rendering|Material|SetMaterial (Variables|Default|GetSM_FaceAcc_Outline) _array_index_3 _array_element_3)))
    (:then_4
      (Rendering|SetVisibility (Variables|Default|GetSM_Body_Outline) (Variables|Debug|GetOutlineMeshVisibility) false)
      (Rendering|SetVisibility (Variables|Default|GetSM_FaceAcc) (Variables|Debug|GetOriginalMeshVisibility) false))))
